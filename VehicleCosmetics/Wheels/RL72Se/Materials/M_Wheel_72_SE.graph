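# Unlit, fully emissive material for the RL72 SE vehicle wheel, tinted by CustomColor and passed through vehicle damage and de-rez transition functions.
# One look combines a polar-mapped scrolling gradient over the rim texture with ShaderMask areas that blink on a timer.
# The other layers Fresnel, a panning normal map and scrolling gradient noise.
# The two looks blend, apparently by vertex color red, before GlobalBrightness scales the result.
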Material /VehicleCosmetics/Wheels/RL72Se/Materials/M_Wheel_72_SE
graph {
 properties {
  bAutomaticallySetUsageInEditor=false
  bUseMaterialAttributes=true
  bUsedWithMeshDeformer=true
  bUsedWithSkeletalMesh=true
  pixelDepthOffsetMode=PDOM_Legacy
  shadingModel=MSM_Unlit
  usedShadingModels=Unlit
 }
 nodes [
  n1 = NamedRerouteUsage()
  MaterialFunctionCall(In=MaterialFunctionCall(Material=SetMaterialAttributes(Emissive Color=@n1) {attributeSetTypes [#1]}) {materialFunction=&MF_BaseVehicle_Damage}.Result) {materialFunction=&MF_VehicleDeRez_Transition}
  NamedRerouteDeclaration(None=VertexColor().R) {name=VC_R, nodeColor=<rgba 1 0 0.0007 1>}
  n2 = Constant() {r=0.5}
  n3 = TextureSampleParameter2D() {parameterName=VC_Mask, samplerType=SAMPLERTYPE_LinearColor, texture=/VehicleCosmetics/Wheels/RL72Se/Textures/RLWH072_Wheel_VC_M}
  n4 = Clamp(None=Add(A=@n3.G, B=Constant() {r=0.2}))
  n5 = ScalarParameter() {defaultValue=3, group=Emissive, parameterName=ColorOffset}
  n6 = Subtract(A=TextureCoordinate() {coordinateIndex=1, uTiling=2, vTiling=2}, B=Constant() {r=1})
  n7 = Arctangent2!(Y=ComponentMask(None=@n6) {g=true}, X=ComponentMask(None=@n6) {r=true})
  n8 = Constant() {r=3.1415}
  n9 = Constant() {r=2}
  n10 = Frac(None=Divide(A=ComponentMask(None=@n7) {r=true}, B=Multiply(A=@n8, B=@n9)))
  n11 = Multiply(A=@n6, B=@n6)
  n12 = SquareRoot(None=Add(A=ComponentMask(None=@n11) {r=true}, B=ComponentMask(None=@n11) {g=true}))
  n13 = TextureSample(UVs=Panner(Coordinate=AppendVector(A=@n10, B=@n12)) {speedY=-0.5}) {texture=/VehicleCosmetics/Wheels/_SharedTextures/FX/GradientStripe2_T}
  n14 = Constant() {r=0.6}
  n15 = Multiply(
   A=Multiply(A=Multiply(A=@n4, B=@n5), B=Power(Base=@n13.RGB, Exp=@n14))
   B=Power(Base=TextureSampleParameter2D() {group=Textures, parameterName=BC, texture=/VehicleCosmetics/Wheels/RL72Se/Textures/T_RL72SE_RIM_BC}.RGB)
  )
  n16 = Time()
  n17 = Constant() {r=0.25}
  n18 = Multiply(A=Frac(None=Multiply(A=@n16, B=@n17)), B=Constant() {r=2})
  n19 = Constant() {r=-1}
  n20 = Add(A=Add(B=Constant() {r=-1}), B=Abs(None=Add(A=@n18, B=@n19)))
  n21 = Constant() {r=1}
  n22 = Multiply(
   A=TextureSampleParameter2D() {parameterName=ShaderMask, samplerType=SAMPLERTYPE_LinearColor, texture=/VehicleCosmetics/Wheels/RL72Se/Textures/T_RL72SE_M}.R
   B=If(A=Abs(None=@n20), B=Constant() {r=0.1}, "A > B"=Constant(), "A == B"=@n21, "A < B"=@n21)
  )
  n23 = Add(A=@n22, B=@n22)
  n24 = Multiply(
   A=ScalarParameter() {defaultValue=10, group=Emissive, parameterName=SmoothnessDots}
   B=Power(Base=Add(A=@n23, B=@n23), Exp=Constant() {r=0.8})
  )
  n25 = VectorParameter() {group=Emissive, parameterName=CustomColor, defaultValue=<rgba 0.31 0.71 1 1>}
  n26 = Divide(A=@n25.RGB, B=Constant() {r=10})
  n27 = AppendVector(A=AppendVector(A=ComponentMask(None=@n26) {b=true}, B=ComponentMask(None=@n26) {g=true}), B=ComponentMask(None=@n26) {r=true})
  n28 = Multiply(A=@n25.RGB, B=Add(A=Multiply(A=@n2, B=@n15), B=Multiply(A=@n27, B=@n24)))
  n29 = Constant() {r=0.5}
  n30 = MaterialFunctionCall!(Normal, Exponent=Constant() {r=2}) {materialFunction=/VehicleCosmetics/Wheels/Cola_Holographic/Materials/MF_Fresnel_Classic}
  n31 = TextureCoordinate()
  n32 = MaterialFunctionCall(
   Normal=TextureSample(UVs=Panner(Coordinate=@n31) {speedX=0.05}) {samplerType=SAMPLERTYPE_Normal, texture=/VehicleCosmetics/Wheels/RL72Se/Textures/T_Citrus_N}.RGB
   Exponent=Constant() {r=2}
  ) {materialFunction=&MF_Fresnel_Classic}
  n33 = Constant() {r=30}
  n34 = Panner(Coordinate=TextureCoordinate() {uTiling=12}) {speedX=-1}
  n35 = TextureCoordinate() {uTiling=0.5}
  n36 = Multiply(
   A=ConstantBiasScale(None=TextureSample(UVs=@n34) {texture=/VehicleCosmetics/Wheels/_SharedTextures/FX/LinearGradient_Balanced_Vertical}.RGB) {bias=-0.2, scale=10}
   B=TextureSample(UVs=Panner(Coordinate=@n35) {speedY=0.5}) {texture=/VehicleCosmetics/Wheels/_SharedTextures/FX/Universal_Noise_34_RGB}.R
  )
  n37 = Add(
   A=LinearInterpolate(A=@n27, B=Constant3Vector(), Alpha=Multiply(A=@n32.Result, B=@n33))
   B=Multiply(A=Multiply(A=@n36, B=@n30.Result), B=Constant() {r=1})
  )
  n38 = Multiply(A=@n25.RGB, B=Add(A=Multiply(A=@n30.Result, B=@n29), B=Clamp(None=@n37)))
  n39 = NamedRerouteUsage()
  NamedRerouteDeclaration(None=Multiply(A=LinearInterpolate(A=@n28, B=@n38, Alpha=@n39), B=ScalarParameter() {defaultValue=1, group=Emissive, parameterName=GlobalBrightness})) {name=Emissive, nodeColor=<rgba 0.1792 0.3686 1 1>}
 ]
}
